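# Translucent UI material that draws a scrolling compass strip for the HUD.
# It samples the alpha of T_Compass through half-tiled U coordinates, scaled about the center and offset by CompassPan and CompassYOffset parameters.
# A horizontal falloff from the middle, shaped by a power and SmoothStep, fades the strip toward both sides.
# An inverted ScaleUVsByCenter mask, sized horizontally by CenterMask, cuts out the central band.
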Material /Game/Athena/HUD/Compass/M_Compass
graph {
 properties {blendMode=BLEND_Translucent, materialDomain=MD_UI, pixelDepthOffsetMode=PDOM_Legacy, usedShadingModels=Default Lit}
 nodes [
  n1 = TextureCoordinate()
  n2 = AppendVector(A=ScalarParameter() {parameterName=CenterMask}, B=Constant() {r=1})
  n3 = Add(
   A=MaterialFunctionCall(UVs=TextureCoordinate() {uTiling=0.5}, Texture Scale=VectorParameter() {parameterName="Texture Scale (V2)", defaultValue=<rgba 1 1 0 0>}.RGB) {materialFunction=&ScaleUVsByCenter}.UVs
   B=MaterialFunctionCall(X=ScalarParameter() {parameterName=CompassPan, sliderMax=10, sliderMin=-10}, Y=ScalarParameter() {parameterName=CompassYOffset}) {materialFunction=&MakeFloat2}.Result
  )
  n4 = ComponentMask(None=TextureCoordinate()) {r=true}
  n5 = OneMinus(None=Abs(None=Add(A=@n4) {constB=-0.5}))
  n6 = Clamp(None=Multiply(A=Power(Base=@n5) {constExponent=6}) {constB=1.5})
  n7 = Constant()
  n8 = Constant() {r=0.4}
  Multiply(
   A=OneMinus(None=MaterialFunctionCall(UVs=@n1, Texture Scale=@n2) {materialFunction=&ScaleUVsByCenter}.0-1 mask)
   B=Multiply(
    A=TextureSample(UVs=@n3) {texture=&/T_Compass}.A
    B=MaterialFunctionCall(Alpha=@n6, Min=@n7, Max=@n8) {materialFunction=&SmoothStep}.Result
   )
  )
 ]
}
