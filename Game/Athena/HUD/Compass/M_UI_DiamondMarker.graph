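# M_UI_DiamondMarker is a translucent UI material for a diamond-shaped marker on the Athena HUD compass, with an animated pulse.
# The shared MF_DiamondPulse function drives the pulse, with parameters for speed, scale, end width and falloff.
# An inner diamond mask comes from rotating UVs an eighth turn and scaling them about the center by InnerDiamondScale, turning the square 0-1 mask into a diamond.
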Material /Game/Athena/HUD/Compass/M_UI_DiamondMarker
graph {
 properties {blendMode=BLEND_Translucent, materialDomain=MD_UI, pixelDepthOffsetMode=PDOM_Legacy, usedShadingModels=Default Lit}
 nodes [
  Constant() {r=1}
  n1 = Constant() {r=0.125}
  MaterialFunctionCall!(
   PulseSpeed
   ScaleStart=ScalarParameter() {parameterName="ScaleStart (S)"}
   ScaleEnd=ScalarParameter() {defaultValue=0.8095, parameterName=EndScale}
   WidthStart
   WidthEnd=ScalarParameter() {parameterName="WidthEnd (S)"}
   PulseAnimFalloff=ScalarParameter() {parameterName="PulseAnimFalloff (S)"}
   Rotation=@n1
   TimeOffset
   ManualTime
   OpacityStart
   OpacityEnd
   TimeSpeed=ScalarParameter() {defaultValue=2, parameterName=PulseSpeed}
   TexCoord
  ) {materialFunction=/Game/Athena/HUD/LTM/Infiltration/Materials/MF_DiamondPulse}
  n2 = TextureCoordinate()
  Max(B=MaterialFunctionCall(
    UVs=MaterialFunctionCall(UVs=@n2, "Rotation Angle (0-1)"=@n1) {materialFunction=&CustomRotator}.Rotated Values
    Texture Scale=ScalarParameter() {defaultValue=0.5, parameterName=InnerDiamondScale}
   ) {materialFunction=&ScaleUVsByCenter}.0-1 mask)
 ]
}
